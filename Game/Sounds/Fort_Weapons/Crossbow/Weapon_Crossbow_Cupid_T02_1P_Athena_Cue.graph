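# First-person firing sound cue for the Cupid crossbow (T02, Athena), using close-shot waves.
# A quality-level node scales variety by branch: random of three waves, random of two, then a single wave shared by the last two.
# Each branch passes through a mixer that lowers volume and has an empty second input, offset by a boosted cue multiplier.
# First-person weapon attenuation, fire sound class and gunfire concurrency govern playback.
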SoundCue /Game/Sounds/Fort_Weapons/Crossbow/Weapon_Crossbow_Cupid_T02_1P_Athena_Cue
root=QualityLevel_1
nodes [
 {id=Mixer_0, children [Random_2,null], properties {inputVolume [0.6,0.6]}}
 {id=Mixer_1, children [Random_1,null], properties {inputVolume [0.6,0.6]}}
 {id=Mixer_3, children [WavePlayer_9,null], properties {inputVolume [0.6,0.6]}}
 {id=QualityLevel_1, children [Mixer_1,Mixer_0,Mixer_3,Mixer_3]}
 {id=Random_1, children [WavePlayer_2,WavePlayer_3,WavePlayer_4], properties {weights [1,1,1]}}
 {id=Random_2, children [WavePlayer_5,WavePlayer_6], properties {weights [1,1]}}
 {id=WavePlayer_2, properties {soundWaveAssetPtr=&/sfx_crossbow_cupid_shot_close_03}}
 {id=WavePlayer_3, properties {soundWaveAssetPtr=&/sfx_crossbow_cupid_shot_close_01}}
 {id=WavePlayer_4, properties {soundWaveAssetPtr=&/sfx_crossbow_cupid_shot_close_02}}
 {id=WavePlayer_5, properties {soundWaveAssetPtr=&sfx_crossbow_cupid_shot_close_03}}
 {id=WavePlayer_6, properties {soundWaveAssetPtr=&sfx_crossbow_cupid_shot_close_01}}
 {id=WavePlayer_9, properties {soundWaveAssetPtr=&sfx_crossbow_cupid_shot_close_01}}
]
properties {
 attenuationSettings=&Fort_Weapon_Projectile_FirstPerson_Athena_Att
 maxDistance=2097152
 soundClassObject=&Fort_WeaponRanged_Fire_1P_PSM
 volumeMultiplier=1.8
 attenuationOverrides {falloffDistance=5000, attenuationShapeExtents {x=1000}}
 concurrencyOverrides {resolutionRule=PreventNew}
 concurrencySet [&Weapon_Gunfire_FirstPerson_Conc]
}
