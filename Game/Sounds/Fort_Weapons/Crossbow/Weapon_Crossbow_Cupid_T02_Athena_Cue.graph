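# Sound cue for the Cupid crossbow (T02, Athena) firing shot, layered by listener distance.
# A mixer blends close, medium and distant branches, with the close layer weighted loudest, each using its own attenuation and sound class override.
# Each branch randomly picks a wave variant; only the close layer uses Cupid-specific samples, and the distant layer adds pitch modulation.
# Medium and distant randomizers opt out of branch culling, and gunfire concurrency sets limit simultaneous instances.
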SoundCue /Game/Sounds/Fort_Weapons/Crossbow/Weapon_Crossbow_Cupid_T02_Athena_Cue
root=Mixer_2
nodes [
 {id=Attenuation_5, children [Random_4], properties {attenuationSettings=&Fort_Weapon_Projectile_Close_Athena_Att}}
 {
  id=Attenuation_6
  children [Random_5]
  properties {attenuationSettings=&Fort_Weapon_Projectile_Medium_Mono_Athena_Att}
 }
 {
  id=Attenuation_7
  children [Modulator_1]
  properties {attenuationSettings=&Fort_Weapon_Projectile_Distant_Mono_Athena_Att}
 }
 {id=Mixer_2, children [SoundClass_4,SoundClass_5,SoundClass_8], properties {inputVolume [1.8,0.5,0.3]}}
 {id=Modulator_1, children [Random_6], properties {pitchMax=1.1}}
 {id=Random_4, children [WavePlayer_8, WavePlayer_14, WavePlayer_15], properties {weights [1,1,1]}}
 {
  id=Random_5
  children [WavePlayer_16, WavePlayer_17, WavePlayer_18, WavePlayer_19]
  properties {bShouldExcludeFromBranchCulling=true, weights [1,1,1,1]}
 }
 {id=Random_6, children [WavePlayer_20, WavePlayer_21, WavePlayer_22], properties {bShouldExcludeFromBranchCulling=true, weights [1,1,1]}}
 {
  id=SoundClass_4
  children [Attenuation_5]
  properties {soundClassOverride=&Fort_WeaponRanged_Fire_3P_PSM}
 }
 {
  id=SoundClass_5
  children [Attenuation_6]
  properties {soundClassOverride=&Fort_WeaponRanged_Fire_Distant}
 }
 {
  id=SoundClass_8
  children [Attenuation_7]
  properties {soundClassOverride=&Fort_WeaponRanged_Fire_Distant}
 }
 {id=WavePlayer_14, properties {soundWaveAssetPtr=&/sfx_crossbow_cupid_shot_close_01}}
 {id=WavePlayer_15, properties {soundWaveAssetPtr=&/sfx_crossbow_cupid_shot_close_02}}
 {id=WavePlayer_16, properties {soundWaveAssetPtr=&/Crossbow_Medium04}}
 {id=WavePlayer_17, properties {soundWaveAssetPtr=&/Crossbow_Medium01}}
 {id=WavePlayer_18, properties {soundWaveAssetPtr=&/Crossbow_Medium02}}
 {id=WavePlayer_19, properties {soundWaveAssetPtr=&/Crossbow_Medium03}}
 {id=WavePlayer_20, properties {soundWaveAssetPtr=&/Crossbow_Distant04}}
 {id=WavePlayer_21, properties {soundWaveAssetPtr=&/Crossbow_Distant02}}
 {id=WavePlayer_22, properties {soundWaveAssetPtr=&/Crossbow_Distant01}}
 {id=WavePlayer_8, properties {soundWaveAssetPtr=&/sfx_crossbow_cupid_shot_close_03}}
]
properties {
 maxDistance=27500
 volumeMultiplier=0.7
 attenuationOverrides {falloffDistance=5000, attenuationShapeExtents {x=1000}}
 concurrencyOverrides {resolutionRule=PreventNew}
 concurrencySet [&Weapon_Gunfire_Local_Conc, &Weapon_Gunfire_Global_Conc]
}
