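# Sound cue for the end of the crossbow reload in the weapon foley sound class.
# A single wave player plays the sfx_crossbow_reload_04 sample directly from the root, with no randomization or modulation.
# The cue applies the shared weapon reload attenuation at reduced volume, overrides concurrency, and has its debug flag enabled.
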SoundCue /Game/Sounds/Fort_Weapons/Crossbow/Weapon_Crossbow_Reload_End_Cue
root=WavePlayer_27
nodes [{id=WavePlayer_27, properties {soundWaveAssetPtr=&/sfx_crossbow_reload_04}}]
properties {
 attenuationSettings=&Fort_Weapon_Reload_Att
 bDebug=true
 bOverrideConcurrency=true
 maxDistance=2250
 soundClassObject=&Fort_WeaponFoley
 volumeMultiplier=0.702
}
