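# First-person close shot sound cue for the T02 crossbow in Athena, using first-person fire sound class, attenuation and gunfire concurrency.
# A quality-level root selects mixer branches with decreasing variation: random of three shots, random of two, then one fixed shot shared by the last two levels.
# Only the first branch overrides to the third-person fire sound class and close projectile attenuation.
# Each mixer has an empty second input.
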SoundCue /Game/Sounds/Fort_Weapons/Crossbow/Weapon_Crossbow_T02_Athena_1P_Cue
root=QualityLevel_0
nodes [
 {id=Attenuation_3, children [Random_11], properties {attenuationSettings=&Fort_Weapon_Projectile_Close_Athena_Att}}
 {id=Mixer_0, children [Random_12,null], properties {inputVolume [0.6,0.6]}}
 {id=Mixer_1, children [SoundClass_3,null], properties {inputVolume [0.6,0.6]}}
 {id=Mixer_3, children [WavePlayer_27, null], properties {inputVolume [0.6,0.6]}}
 {id=QualityLevel_0, children [Mixer_1,Mixer_0,Mixer_3,Mixer_3]}
 {id=Random_11, children [WavePlayer_21, WavePlayer_22, WavePlayer_23], properties {weights [1,1,1]}}
 {id=Random_12, children [WavePlayer_24, WavePlayer_25], properties {weights [1,1]}}
 {
  id=SoundClass_3
  children [Attenuation_3]
  properties {soundClassOverride=&Fort_WeaponRanged_Fire_3P_PSM}
 }
 {id=WavePlayer_21, properties {soundWaveAssetPtr=&/sfx_crossbow_shot_close_05}}
 {id=WavePlayer_22, properties {soundWaveAssetPtr=&/sfx_crossbow_shot_close_06}}
 {id=WavePlayer_23, properties {soundWaveAssetPtr=&/sfx_crossbow_shot_close_07}}
 {id=WavePlayer_24, properties {soundWaveAssetPtr=&sfx_crossbow_shot_close_05}}
 {id=WavePlayer_25, properties {soundWaveAssetPtr=&sfx_crossbow_shot_close_06}}
 {id=WavePlayer_27, properties {soundWaveAssetPtr=&sfx_crossbow_shot_close_05}}
]
properties {
 attenuationSettings=&Fort_Weapon_Projectile_FirstPerson_Athena_Att
 maxDistance=2097152
 soundClassObject=&Fort_WeaponRanged_Fire_1P_PSM
 volumeMultiplier=0.7
 attenuationOverrides {falloffDistance=5000, attenuationShapeExtents {x=1000}}
 concurrencyOverrides {resolutionRule=PreventNew}
 concurrencySet [&Weapon_Gunfire_FirstPerson_Conc]
}
